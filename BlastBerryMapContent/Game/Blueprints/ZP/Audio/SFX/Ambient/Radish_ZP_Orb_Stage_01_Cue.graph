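# Looping ambient sound cue for the stage 01 Radish ZP orb in the BlastBerry map.
# A single looping wave player plays a temp-audio wave at reduced volume on the Athena ambient props sound class.
# It references the shared Radish ZP attenuation asset and also carries override values with listener focus, which favor the sound when the listener faces it.
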SoundCue /BlastBerryMapContent/Game/Blueprints/ZP/Audio/SFX/Ambient/Radish_ZP_Orb_Stage_01_Cue
root=WavePlayer_0
nodes [{
  id=WavePlayer_0
  properties {bLooping=true, soundWaveAssetPtr=/BlastberryMapContent/Game/Blueprints/ZP/Audio/TempAudio/Ambient/Radish_ZP_Orb_Stage_01}
 }]
properties {
 attenuationSettings=/BlastBerryMapContent/Game/Blueprints/ZP/Audio/Settings/Radish_ZP_Att
 maxDistance=13000
 soundClassObject=&Fort_Ambient_Props_Athena
 volumeMultiplier=0.501
 attenuationOverrides {
  bEnableFocusInterpolation=true
  bEnableListenerFocus=true
  falloffDistance=4000
  focusReleaseInterpSpeed=3
  nonFocusAzimuth=90
  nonFocusDistanceScale=2
  nonFocusPriorityScale=0.5
  nonFocusVolumeAttenuation=0.3
  attenuationShapeExtents {x=1000}
 }
}
